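# Conversation database for the Love Ranger NPC, defining its greeting dialogue in spooky, foreboding flavor.
# A single entry point leads to a speech node with a default line plus prioritized contextual lines for first-ever meetings and for specific player outfits checked by character ID.
# The speech node then branches through link nodes to the shared NPC quests and basic services conversations and a Season 26 Snapshot quest conversation.
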ConversationDatabase /NPCLibrary/NPCs/LoveRanger/NPCLibrary_Conversation_LoveRanger
entry_points [{tag=Conversation.Entry.LoveRanger, destinations [#1]}]
nodes [
 ConversationEntryPointNode_7 {
  id=#1
  class=ConversationEntryPointNode
  outputs [#2]
  properties {entryTag {tagName=Conversation.Entry.LoveRanger}}
 }
 Speech_C_4 {
  id=#2
  class=Speech_C
  outputs [#3, #4, #5]
  properties {
   nodeName=""
   generalConfig {
    defaultMessage=Have you seen the creeping DARKNESS on the horizon too??
    contextualMessages [
     {
      message="Be careful, friend. The spooky evening of NIGHTMARES is coming!!"
      contextRequirements [{requirement {class=FortControllerRequirement_IsFirstEverConversationWithNPC}}]
      priority {curve {rowName=Tandem.Speech.Priority.FirstEverInteraction}}
     }
     {
      message=I see in your spooky future... a XYLOPHONE of bones...
      contextRequirements [{requirement {
         class=FortControllerRequirement_HasCID
         properties {allowedCIds [&CID_701_Athena_Commando_M_BananaAgent]}
        }}]
      priority {curve {rowName=Tandem.Speech.Priority.OutfitReaction}}
     }
     {
      message="I see in your spooky future... bright eyes enveloped in SHADOW!"
      contextRequirements [{requirement {
         class=FortControllerRequirement_HasCID
         properties {allowedCIds [
           &CID_693_Athena_Commando_M_BuffCat
           /BRCosmetics/Athena/Items/Cosmetics/Characters/Character_BuffCatCruise
          ]}
        }}]
      priority {curve {rowName=Tandem.Speech.Priority.OutfitReaction}}
     }
     {
      message="I see in your spooky future... a murky swamp, and SOMETHING within!"
      contextRequirements [{requirement {
         class=FortControllerRequirement_HasCID
         properties {allowedCIds [&CID_315_Athena_Commando_M_TeriyakiFish]}
        }}]
      priority {curve {rowName=Tandem.Speech.Priority.OutfitReaction}}
     }
     {
      message="I can see my own spooky future... and it is wreathed in FLAME!"
      contextRequirements [{requirement {
         class=FortControllerRequirement_HasCID
         properties {allowedCIds [/BRCosmetics/Athena/Items/Cosmetics/Characters/CID_523_Athena_Commando_F_Cupid]}
        }}]
      priority {curve {rowName=Tandem.Speech.Priority.OutfitReaction}}
     }
     {
      message="I can see the spooky future, and the island's spirits will soon be seeking your aid!"
      contextRequirements [{requirement {
         class=FortControllerRequirement_HasCID
         properties {allowedCIds [/BRCosmetics/Athena/Items/Cosmetics/Characters/CID_A_438_Athena_Commando_F_Gloom]}
        }}]
      priority {curve {rowName=Tandem.Speech.Priority.OutfitReaction}}
     }
    ]
   }
  }
 }
 ConversationLinkNode_2 {
  id=#5
  class=ConversationLinkNode
  properties {remoteEntryTag {tagName=Conversation.Entry.Season26.Snapshot.P03.Q07}}
 }
 ConversationLinkNode_0 {id=#4, class=ConversationLinkNode, properties {remoteEntryTag {tagName=Conversation.Entry.BasicServices}}}
 ConversationLinkNode_1 {id=#3, class=ConversationLinkNode, properties {remoteEntryTag {tagName=Conversation.Entry.NPCLibrary.Quests}}}
]
